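# CapeTest_spa1000_anm is a Level Sequence for an animation shot in the CapeTest test level, featuring the KalEl superhero cinematic NPC.
# One cine camera with constant transform and lens channels carries spawn and attach tracks and holds the single camera cut.
# A keyed visibility track toggles KalEl on, off and on again, roughly timed to the starts of two subsequence sections whose contents are not shown here.
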LevelSequence /PistolCanary/Levels/Test/CapeTest/LevelSequences/CapeTest_spa1000_anm
bindings [
 CineCameraActor {
  id=#1
  possessed_class=CineCameraActor
  tracks [
   {
    type=MovieScene3DTransformTrack
    sections [{
      type=MovieScene3DTransformSection
      channels [
       Location.X {type=MovieSceneScriptingDoubleChannel, default=7}
       Location.Y {type=MovieSceneScriptingDoubleChannel, default=-344}
       Location.Z {type=MovieSceneScriptingDoubleChannel, default=588}
       Rotation.X {type=MovieSceneScriptingDoubleChannel, default=90}
       Rotation.Y {type=MovieSceneScriptingDoubleChannel, default=-120}
       Rotation.Z {type=MovieSceneScriptingDoubleChannel, default=-90}
       Scale.X {type=MovieSceneScriptingDoubleChannel, default=1}
       Scale.Y {type=MovieSceneScriptingDoubleChannel, default=1}
       Scale.Z {type=MovieSceneScriptingDoubleChannel, default=1}
      ]
     }]
   }
   {type=MovieSceneSpawnTrack, sections [{type=MovieSceneSpawnSection, channels [None {type=MovieSceneScriptingBoolChannel, default=true}]}]}
   {type=MovieScene3DAttachTrack, sections [{type=MovieScene3DAttachSection, end_frame=9586, start_frame=3600}]}
  ]
 }
 KalEl {
  id=#2
  possessed_class=PistolCanary_CinematicNPC_Superhero_Main_C
  tracks [{
    type=MovieSceneVisibilityTrack
    sections [{
      type=MovieSceneVisibilitySection
      channels [None {type=MovieSceneScriptingBoolChannel, default=false, keys [[6113,false],[6122,true],[6232,false],[6895,true]]}]
     }]
   }]
 }
 CameraComponent {
  id=#3
  parent_id=#1
  possessed_class=CineCameraComponent
  tracks [
   {
    type=MovieSceneFloatTrack
    display_name=Current Focal Length
    sections [{type=MovieSceneFloatSection, channels [None {type=MovieSceneScriptingFloatChannel, default=35}]}]
   }
   {
    type=MovieSceneFloatTrack
    display_name="Manual Focus Distance (Focus Settings)"
    sections [{type=MovieSceneFloatSection, channels [None {type=MovieSceneScriptingFloatChannel, default=100000}]}]
   }
   {
    type=MovieSceneFloatTrack
    display_name=Current Aperture
    sections [{type=MovieSceneFloatSection, channels [None {type=MovieSceneScriptingFloatChannel, default=2.8}]}]
   }
  ]
 }
]
display_rate {denominator=1, numerator=60}
tick_resolution {denominator=1, numerator=24000}
tracks [
 {
  type=MovieSceneSubTrack
  display_name=Subsequences
  sections [{type=MovieSceneSubSection, end_frame=6411, start_frame=6122}, {type=MovieSceneSubSection, end_frame=9296, start_frame=6896}]
 }
 {type=MovieSceneCameraCutTrack, sections [{type=MovieSceneCameraCutSection, end_frame=9586, start_frame=3600}]}
]
