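# Test level sequence for CapeTest that flies the KalEl cinematic superhero NPC rapidly along X.
# Three overlapping skeletal animation sections play while transform and visibility keys drive the flight and briefly hide the character.
# The child Backpack skeletal mesh has a constant Cloth Geometry Scale and an event trigger.
# A child Niagara component's LifetimeScale holds high, then drops near the end of the animation.
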LevelSequence /PistolCanary/Levels/Test/CapeTest/LevelSequences/CapeTest_spa1000b_anm
bindings [
 Backpack {
  id=#1
  parent_id=#2
  possessed_class=SkeletalMeshComponentBudgeted
  tracks [
   {
    type=MovieSceneEventTrack
    sections [{
      type=MovieSceneEventTriggerSection
      channels [None {type=MovieSceneScriptingEventChannel, const ["<Struct 'MovieSceneEvent' (0x85285c430c0) {}>", 1]}]
     }]
   }
   {
    type=MovieSceneFloatTrack
    display_name=Cloth Geometry Scale
    sections [{type=MovieSceneFloatSection, channels [None {type=MovieSceneScriptingFloatChannel, default=0.2886, const [0.2886,1]}]}]
   }
  ]
 }
 KalEl {
  id=#2
  possessed_class=PistolCanary_CinematicNPC_Superhero_Main_C
  tracks [
   {
    type=MovieSceneSkeletalAnimationTrack
    sections [
     {type=MovieSceneSkeletalAnimationSection, end_frame=3720, start_frame=3600}
     {type=MovieSceneSkeletalAnimationSection, end_frame=3818, start_frame=3698}
     {type=MovieSceneSkeletalAnimationSection, end_frame=3923, start_frame=3803}
    ]
   }
   {
    type=MovieSceneVisibilityTrack
    sections [{
      type=MovieSceneVisibilitySection
      channels [None {type=MovieSceneScriptingBoolChannel, default=true, keys [[3604,false],[3626,true],[3834,true]]}]
     }]
   }
   {
    type=MovieScene3DTransformTrack
    sections [{
      type=MovieScene3DTransformSection
      end_frame=7541
      channels [
       Location.X {type=MovieSceneScriptingDoubleChannel, default=346.688, keys [[3624,41556.4883],[3675,346.688],[3751,-66032.5462]]}
       Location.Y {type=MovieSceneScriptingDoubleChannel, default=-91.0374, thinned_from=3, keys [[3600,-91.0372],[3792,-91.0382]]}
       Location.Z {type=MovieSceneScriptingDoubleChannel, default=6632.5984, const [6632.5981,3]}
       Rotation.X {type=MovieSceneScriptingDoubleChannel, default=9.7172, const [9.7172,3]}
       Rotation.Y {type=MovieSceneScriptingDoubleChannel, default=6.0492, const [6.0492,3]}
       Rotation.Z {type=MovieSceneScriptingDoubleChannel, default=447.918, const [447.918,3]}
       Scale.X {type=MovieSceneScriptingDoubleChannel, default=3.4654}
       Scale.Y {type=MovieSceneScriptingDoubleChannel, default=3.4654}
       Scale.Z {type=MovieSceneScriptingDoubleChannel, default=3.4654}
      ]
     }]
   }
  ]
 }
 Niagara {
  id=#3
  parent_id=#2
  possessed_class=NiagaraComponent
  tracks [{
    type=MovieSceneNiagaraFloatParameterTrack
    display_name=LifetimeScale
    sections [{type=MovieSceneFloatSection, channels [None {type=MovieSceneScriptingFloatChannel, default=15, keys [[3619,15],[3854,15],[3887,1]]}]}]
   }]
 }
]
display_rate {denominator=1, numerator=60}
tick_resolution {denominator=1, numerator=24000}
tracks []
