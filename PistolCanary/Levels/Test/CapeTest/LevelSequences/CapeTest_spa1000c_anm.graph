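# Cape test Level Sequence animating the KalEl superhero cinematic NPC, with temporary eye laser effects.
# KalEl plays two back-to-back skeletal animations at fixed location and scale, with a brief rotation turn and late hide.
# Its Backpack component sets cloth geometry scale and fires an event, while its Niagara component's LifetimeScale is keyed.
# Two attached laser meshes flash through keyed visibility, rotation and length scaling, though their spawn tracks default to off.
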LevelSequence /PistolCanary/Levels/Test/CapeTest/LevelSequences/CapeTest_spa1000c_anm
bindings [
 Niagara {
  id=#1
  parent_id=#2
  possessed_class=NiagaraComponent
  tracks [{
    type=MovieSceneNiagaraFloatParameterTrack
    display_name=LifetimeScale
    sections [{
      type=MovieSceneFloatSection
      channels [None {
        type=MovieSceneScriptingFloatChannel
        default=20
        keys [[3600,10],[3707,10],[3881,1],[4076,1],[4086,10],[4268,10],[4345,1],[4405,1],[4445,20],[5092,20],[5195,1],[5413,1],[5423,20]]
       }]
     }]
   }]
 }
 KalEl {
  id=#2
  possessed_class=PistolCanary_CinematicNPC_Superhero_Main_C
  tracks [
   {
    type=MovieScene3DTransformTrack
    sections [{
      type=MovieScene3DTransformSection
      end_frame=7541
      channels [
       Location.X {type=MovieSceneScriptingDoubleChannel, default=-52714, const [495.3745,1]}
       Location.Y {type=MovieSceneScriptingDoubleChannel, default=25000, const [-613.4579,1]}
       Location.Z {type=MovieSceneScriptingDoubleChannel, default=232776.6406, const [32880.7305,1]}
       Rotation.X {type=MovieSceneScriptingDoubleChannel, default=0, keys [[3600,184.4053],[3644,178.8983]]}
       Rotation.Y {type=MovieSceneScriptingDoubleChannel, default=0, keys [[3600,13.4808],[3644,0.7977]]}
       Rotation.Z {type=MovieSceneScriptingDoubleChannel, default=0, keys [[3600,-14.1991],[3644,51.3411]]}
       Scale.X {type=MovieSceneScriptingDoubleChannel, default=4, const [4,1]}
       Scale.Y {type=MovieSceneScriptingDoubleChannel, default=4, const [4,1]}
       Scale.Z {type=MovieSceneScriptingDoubleChannel, default=4, const [4,1]}
      ]
     }]
   }
   {
    type=MovieSceneSkeletalAnimationTrack
    sections [
     {type=MovieSceneSkeletalAnimationSection, end_frame=4776, start_frame=3600}
     {type=MovieSceneSkeletalAnimationSection, end_frame=6000, start_frame=4776}
    ]
   }
   {
    type=MovieSceneVisibilityTrack
    sections [{
      type=MovieSceneVisibilitySection
      channels [None {type=MovieSceneScriptingBoolChannel, default=true, keys [[3600,true],[3610,true],[5966,false]]}]
     }]
   }
  ]
 }
 Eye Laser Temp R {
  id=#3
  possessed_class=FortStaticMeshActor
  tracks [
   {type=MovieSceneSpawnTrack, sections [{type=MovieSceneSpawnSection, channels [None {type=MovieSceneScriptingBoolChannel, default=false}]}]}
   {
    type=MovieScene3DTransformTrack
    sections [{
      type=MovieScene3DTransformSection
      channels [
       Location.X {type=MovieSceneScriptingDoubleChannel, default=0}
       Location.Y {type=MovieSceneScriptingDoubleChannel, default=0}
       Location.Z {type=MovieSceneScriptingDoubleChannel, default=0}
       Rotation.X {type=MovieSceneScriptingDoubleChannel, default=90, const [90,2]}
       Rotation.Y {type=MovieSceneScriptingDoubleChannel, default=0, const [0,2]}
       Rotation.Z {type=MovieSceneScriptingDoubleChannel, default=0, keys [[3916,-90],[3991,0],[4661,0],[4671,-90]]}
       Scale.X {type=MovieSceneScriptingDoubleChannel, default=0.0357, const [0.0357,1]}
       Scale.Y {type=MovieSceneScriptingDoubleChannel, default=0.0357, const [0.0357,1]}
       Scale.Z {
        type=MovieSceneScriptingDoubleChannel
        default=2.56
        keys [[3910,5.2038],[3954,2.0912],[3991,3.0417],[4669,4],[4671,10.365],[4716,3.795],[4849,7.1925],[4888,0.8975],[4966,8.5175],[5016,1.94],[5089,9.4825],[5133,2.56]]
       }
      ]
     }]
   }
   {type=MovieScene3DAttachTrack, sections [{type=MovieScene3DAttachSection, end_frame=12177, start_frame=3836}]}
   {
    type=MovieSceneVisibilityTrack
    sections [{
      type=MovieSceneVisibilitySection
      channels [None {
        type=MovieSceneScriptingBoolChannel
        default=false
        keys [[3836,false],[3910,true],[4079,false],[4671,true],[4717,false],[4722,false],[4849,true],[4889,false],[4966,true],[5017,false],[5089,true],[5134,false]]
       }]
     }]
   }
  ]
 }
 Backpack {
  id=#4
  parent_id=#2
  possessed_class=SkeletalMeshComponentBudgeted
  tracks [
   {
    type=MovieSceneFloatTrack
    display_name=Cloth Geometry Scale
    sections [{type=MovieSceneFloatSection, channels [None {type=MovieSceneScriptingFloatChannel, default=0.25, const [0.25,1]}]}]
   }
   {
    type=MovieSceneEventTrack
    sections [{
      type=MovieSceneEventTriggerSection
      channels [None {type=MovieSceneScriptingEventChannel, const ["<Struct 'MovieSceneEvent' (0x85285c442e0) {}>", 1]}]
     }]
   }
  ]
 }
 Eye Laser Temp L {
  id=#5
  possessed_class=FortStaticMeshActor
  tracks [
   {type=MovieSceneSpawnTrack, sections [{type=MovieSceneSpawnSection, channels [None {type=MovieSceneScriptingBoolChannel, default=false}]}]}
   {
    type=MovieScene3DTransformTrack
    sections [{
      type=MovieScene3DTransformSection
      channels [
       Location.X {type=MovieSceneScriptingDoubleChannel, default=0}
       Location.Y {type=MovieSceneScriptingDoubleChannel, default=0}
       Location.Z {type=MovieSceneScriptingDoubleChannel, default=0}
       Rotation.X {type=MovieSceneScriptingDoubleChannel, default=90, const [90,2]}
       Rotation.Y {type=MovieSceneScriptingDoubleChannel, default=0, const [0,2]}
       Rotation.Z {type=MovieSceneScriptingDoubleChannel, default=0, keys [[3916,-90],[3991,0],[4661,0],[4671,-90]]}
       Scale.X {type=MovieSceneScriptingDoubleChannel, default=0.0357}
       Scale.Y {type=MovieSceneScriptingDoubleChannel, default=0.0357}
       Scale.Z {
        type=MovieSceneScriptingDoubleChannel
        default=-361.2139
        keys [[3910,5.2038],[3954,2.0912],[3991,3.0417],[4669,4],[4671,10.365],[4716,3.795],[4849,7.1925],[4888,0.8975],[4966,8.5175],[5016,1.94],[5089,9.4825],[5133,2.56]]
       }
      ]
     }]
   }
   {type=MovieScene3DAttachTrack, sections [{type=MovieScene3DAttachSection, end_frame=12177, start_frame=3836}]}
   {
    type=MovieSceneVisibilityTrack
    sections [{
      type=MovieSceneVisibilitySection
      channels [None {
        type=MovieSceneScriptingBoolChannel
        default=true
        keys [[3836,false],[3910,true],[4079,false],[4671,true],[4717,false],[4722,true],[4849,true],[4889,false],[4966,true],[5017,false],[5089,true],[5134,false]]
       }]
     }]
   }
  ]
 }
]
display_rate {denominator=1, numerator=60}
tick_resolution {denominator=1, numerator=24000}
tracks []
